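# Physics asset for the F_MED_Rays backpack cosmetic, giving the pack and its four arrows collision and joint limits.
# The pack bone uses a single box body, and each of the arrow_A to arrow_D joints has a thin capsule.
# Each arrow is constrained directly to the pack with tight, equal swing and twist limits, which restricts any simulated arrow motion to a slight wobble.
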PhysicsAsset /BRCosmetics/Accessories/FORT_Backpacks/Backpack_F_MED_Rays/Meshes/F_MED_Rays_Pack_Physics
bodies {
 arrow_A_joint [{
   shapeType=Capsule
   length=30.5369
   radius=0.867
   shapeName=arrow_A_joint_capsule
   center=<-1.6972 -0.0126 -0.0289>
   rotation=<rot 0 90.0003 -89.8084>
  }]
 arrow_B_joint [{
   shapeType=Capsule
   length=33.3032
   radius=0.8536
   shapeName=arrow_B_joint_capsule
   center=<-2.4919 -0.0185 0.0227>
   rotation=<rot 0 -89.8279 -89.912>
  }]
 arrow_C_joint [{
   shapeType=Capsule
   length=28.3254
   radius=0.8687
   shapeName=arrow_C_joint_capsule
   center=<-0.5675 -0.0126 -0.0251>
   rotation=<rot 0 90.0003 -89.8084>
  }]
 arrow_D_joint [{
   shapeType=Capsule
   length=31.0463
   radius=0.804
   shapeName=arrow_D_joint_capsule
   center=<-2.251 -0.0182 0.021>
   rotation=<rot 0 -89.8279 -89.9205>
  }]
 pack [{
   shapeType=Box
   extentX=10.9666
   extentY=4.8393
   extentZ=46.7278
   shapeName=pack_box
   center=<20.8071 -0.8956 -35.5403>
   rotation=<rot 36.1282 -0.0006 -0.0011>
  }]
}
constraints [
 {bone1Name=arrow_A_joint, bone2Name=pack, swing1LimitDegrees=5, swing2LimitDegrees=5, twistLimitDegrees=5}
 {bone1Name=arrow_B_joint, bone2Name=pack, swing1LimitDegrees=5, swing2LimitDegrees=5, twistLimitDegrees=5}
 {bone1Name=arrow_C_joint, bone2Name=pack, swing1LimitDegrees=5, swing2LimitDegrees=5, twistLimitDegrees=5}
 {bone1Name=arrow_D_joint, bone2Name=pack, swing1LimitDegrees=5, swing2LimitDegrees=5, twistLimitDegrees=5}
]
